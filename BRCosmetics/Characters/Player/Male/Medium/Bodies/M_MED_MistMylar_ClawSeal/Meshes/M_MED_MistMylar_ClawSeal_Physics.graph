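# Physics asset for the MistMylar ClawSeal medium male player body cosmetic, setting up physics for its back and right-side dyn_mask bone chains.
# Each mask bone has a large offset capsule and is constrained to a tiny-capsule master bone, with matching asymmetric swing limits and tight twist.
# Pelvis and right thigh capsules are the only body shapes, suggesting the masks hang at the hip and collide there.
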PhysicsAsset /BRCosmetics/Characters/Player/Male/Medium/Bodies/M_MED_MistMylar_ClawSeal/Meshes/M_MED_MistMylar_ClawSeal_Physics
bodies {
 dyn_mask_bk [{shapeType=Capsule, length=6.8587, radius=6.3814, shapeName=dyn_mask_bk_capsule, center=<8.5 0 6>, rotation=<rot 90 0 0>}]
 dyn_mask_bk_mstr [{shapeType=Capsule, length=0.505, radius=0.505, shapeName=dyn_mask_bk_mstr_capsule, rotation=<rot 0 -90 0>}]
 dyn_mask_r [{shapeType=Capsule, length=5.9016, radius=7.8585, shapeName=dyn_mask_r_capsule, center=<10.2179 0 8>, rotation=<rot 80 -0.0004 0.0001>}]
 dyn_mask_r_mstr [{shapeType=Capsule, length=0.505, radius=0.505, shapeName=dyn_mask_r_mstr_capsule, rotation=<rot 0 -90 0>}]
 pelvis [{shapeType=Capsule, length=15.317, radius=11.3012, center=<-6 0 0>}]
 thigh_r [{shapeType=Capsule, length=28.0289, radius=10.9129, shapeName=thigh_r_capsule, center=<17.7413 0 0.6565>, rotation=<rot 90 0 0>}]
}
constraints [
 {bone1Name=dyn_mask_bk, bone2Name=dyn_mask_bk_mstr, swing1LimitDegrees=20, swing2LimitDegrees=70, twistLimitDegrees=6}
 {bone1Name=dyn_mask_r, bone2Name=dyn_mask_r_mstr, swing1LimitDegrees=20, swing2LimitDegrees=70, twistLimitDegrees=6}
]
