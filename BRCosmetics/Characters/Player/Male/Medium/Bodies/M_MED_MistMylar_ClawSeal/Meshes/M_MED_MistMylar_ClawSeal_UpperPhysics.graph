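# Physics asset for the upper mesh of the MistMylar ClawSeal medium male body cosmetic, covering two dangling pullstrings and the torso.
# Each pullstring is a capsule chain from a master bone through three segments, with limited swing and twist, tightest swing at the first joint.
# Large capsules on spine_03 to spine_05 give the torso volume.
# Both masters and the upper spine bodies constrain to spine_03 with free angular motion.
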PhysicsAsset /BRCosmetics/Characters/Player/Male/Medium/Bodies/M_MED_MistMylar_ClawSeal/Meshes/M_MED_MistMylar_ClawSeal_UpperPhysics
bodies {
 dyn_pullstring_l_01 [{
   shapeType=Capsule
   length=3.7341
   radius=0.7
   shapeName=dyn_pullstring_l_01_capsule
   center=<2.59 -0.0402 0.3687>
   rotation=<rot 0 89.1099 -98.3287>
  }]
 dyn_pullstring_l_02 [{
   shapeType=Capsule
   length=4.4726
   radius=0.7765
   shapeName=dyn_pullstring_l_02_capsule
   center=<1.9404 -0.006 -0.2572>
   rotation=<rot 0 -89.5255 -75.3834>
  }]
 dyn_pullstring_l_03 [{
   shapeType=Capsule
   length=5.3808
   radius=0.7025
   shapeName=dyn_pullstring_l_02_capsule
   center=<2.9942 0.0046 0.2001>
   rotation=<rot 0 -89.5255 -85.8425>
  }]
 dyn_pullstring_l_mstr [{shapeType=Capsule, length=0.505, radius=0.505, shapeName=dyn_pullstring_l_mstr_capsule, rotation=<rot 0 -90 0>}]
 dyn_pullstring_r_01 [{
   shapeType=Capsule
   length=5.8314
   radius=1.1002
   shapeName=dyn_pullstring_r_01_capsule
   center=<3.6656 -0.0203 1.3209>
   rotation=<rot -0.001 -89.3944 -85.0275>
  }]
 dyn_pullstring_r_02 [{
   shapeType=Capsule
   length=6.3136
   radius=1.1593
   shapeName=dyn_pullstring_r_02_capsule
   center=<2.4682 -0.0959 0.2966>
   rotation=<rot 0 89.573 -107.3896>
  }]
 dyn_pullstring_r_03 [{shapeType=Capsule, length=5.3768, radius=1.1589, shapeName=dyn_pullstring_r_03_capsule, center=<2.5587 0.0546 0.681>, rotation=<rot 90 0 0>}]
 dyn_pullstring_r_mstr [{shapeType=Capsule, length=0.505, radius=0.505, shapeName=dyn_pullstring_r_mstr_capsule, rotation=<rot 0 -90 0>}]
 spine_03 [{
   shapeType=Capsule
   length=31.575
   radius=11.0523
   shapeName=spine_03_capsule
   center=<9.518 1.0449 0.0719>
   rotation=<rot -1.3508 -90 0.8523>
  }]
 spine_04 [{shapeType=Capsule, length=30.0045, radius=12.2781, shapeName=spine_04_capsule, center=<4.6541 0 0>}]
 spine_05 [{shapeType=Capsule, length=33.5179, radius=10.4292, shapeName=spine_05_capsule, center=<-0.2099 0.948 0>}]
}
constraints [
 {bone1Name=dyn_pullstring_r_03, bone2Name=dyn_pullstring_r_02, swing1LimitDegrees=45, swing2LimitDegrees=45, twistLimitDegrees=6}
 {bone1Name=dyn_pullstring_l_01, bone2Name=dyn_pullstring_l_mstr, swing1LimitDegrees=24, swing2LimitDegrees=24, twistLimitDegrees=8}
 {bone1Name=dyn_pullstring_l_02, bone2Name=dyn_pullstring_l_01, swing1LimitDegrees=45, swing2LimitDegrees=45, twistLimitDegrees=8}
 {bone1Name=dyn_pullstring_r_01, bone2Name=dyn_pullstring_r_mstr, swing1LimitDegrees=24, swing2LimitDegrees=24, twistLimitDegrees=6}
 {bone1Name=dyn_pullstring_r_02, bone2Name=dyn_pullstring_r_01, swing1LimitDegrees=45, swing2LimitDegrees=45, twistLimitDegrees=6}
 {bone1Name=dyn_pullstring_l_03, bone2Name=dyn_pullstring_l_02, swing1LimitDegrees=45, swing2LimitDegrees=45, twistLimitDegrees=8}
 {
  bone1Name=spine_04
  bone2Name=spine_03
  swing1LimitDegrees=45
  swing1Motion=Free
  swing2LimitDegrees=45
  swing2Motion=Free
  twistLimitDegrees=45
  twistMotion=Free
 }
 {
  bone1Name=spine_05
  bone2Name=spine_03
  swing1LimitDegrees=45
  swing1Motion=Free
  swing2LimitDegrees=45
  swing2Motion=Free
  twistLimitDegrees=45
  twistMotion=Free
 }
 {
  bone1Name=dyn_pullstring_r_mstr
  bone2Name=spine_03
  swing1LimitDegrees=45
  swing1Motion=Free
  swing2LimitDegrees=45
  swing2Motion=Free
  twistLimitDegrees=45
  twistMotion=Free
 }
 {
  bone1Name=dyn_pullstring_l_mstr
  bone2Name=spine_03
  swing1LimitDegrees=45
  swing1Motion=Free
  swing2LimitDegrees=45
  swing2Motion=Free
  twistLimitDegrees=45
  twistMotion=Free
 }
]
