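# Animation Blueprint for the MistMylar ClawSeal male body cosmetic, adding secondary motion to the pose copied from its attached parent.
# The pose passes through a Control Rig, then two RigidBody simulations based at pelvis and spine_05.
# Weighted constraints drive zipper master bones from leg and pelvis bones, then sprung AnimDynamics nodes move zippers, loops and collar.
# Each update, a deprecated state function sets gravity override and RigidBody settings from per-movement-state presets.
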AnimBlueprint /BRCosmetics/Characters/Player/Male/Medium/Bodies/M_MED_MistMylar_ClawSeal/Meshes/M_MED_MistMylar_ClawSeal_AnimBP
target_skeleton=&/M_MED_MistMylar_ClawSeal_Skeleton
graphs [
 AnimGraph {nodes [
   n1 = Control Rig(Source=Copy Pose From Mesh() {bUseAttachedParent=true}) {
    alphaBoolBlend {blendOption=Linear}
    controlRigAssetReference {blueprintRigClass=&/M_MED_MistMylar_ClawSeal_CtrlRig.M_MED_MistMylar_ClawSeal_CtrlRig_C}
   }
   n2 = RigidBody!(
    ComponentPose=Local To Component(LocalPose=@n1)
    Alpha
    OverrideWorldGravity:Vector="0,0,0"
    ExternalForce
    ComponentLinearAccScale:Vector="0.3,0.3,0.5"
    ComponentLinearVelScale:Vector="0.4,0.4,0.6"
    SimSpaceSettings:Sim Space Settings Structure=(WorldAlpha=0.75,VelocityScaleZ=1,DampingAlpha=1,MaxLinearVelocity=300,MaxAngularVelocity=5,MaxLinearAcceleration=150,MaxAngularAcceleration=3,ExternalLinearDragV=(X=0,Y=0,Z=0),ExternalLinearVelocity=(X=0,Y=0,Z=0),ExternalAngularVelocity=(X=0,Y=0,Z=0))
   ) -> (Pose) {
    bClampLinearTranslationLimitToRefPose=true
    bOverrideWorldGravity=true
    overridePhysicsAsset=&/M_MED_MistMylar_ClawSeal_Physics
    simulationSpace=BaseBoneSpace
    alphaBoolBlend {blendOption=Linear}
    alphaScaleBiasClamp {interpSpeedDecreasing=6, interpSpeedIncreasing=7}
    baseBoneRef {boneName=pelvis}
    componentAppliedLinearAccClamp=<7500 7500 7500>
    componentLinearAccScale=<0.3 0.3 0.5>
    componentLinearVelScale=<0.4 0.4 0.6>
    overrideWorldGravity {z=-980}
    simSpaceSettings {maxAngularAcceleration=3, maxAngularVelocity=5, maxLinearAcceleration=150, maxLinearVelocity=300, worldAlpha=0.75}
   }
   n3 = RigidBody(
    ComponentPose=@n2
    OverrideWorldGravity="0,0,0"
    ComponentLinearAccScale="0.3,0.3,0.3"
    ComponentLinearVelScale="0.4,0.4,0.4"
    SimSpaceSettings=(WorldAlpha=0.5,VelocityScaleZ=1,DampingAlpha=1,MaxLinearVelocity=300,MaxAngularVelocity=10,MaxLinearAcceleration=150,MaxAngularAcceleration=7,ExternalLinearDragV=(X=0,Y=0,Z=0),ExternalLinearVelocity=(X=0,Y=0,Z=0),ExternalAngularVelocity=(X=0,Y=0,Z=0))
   ) {
    bClampLinearTranslationLimitToRefPose=true
    bOverrideWorldGravity=true
    overridePhysicsAsset=&/M_MED_MistMylar_ClawSeal_UpperPhysics
    simulationSpace=BaseBoneSpace
    alphaBoolBlend {blendOption=Linear}
    alphaScaleBiasClamp {interpSpeedDecreasing=6, interpSpeedIncreasing=7}
    baseBoneRef {boneName=spine_05}
    componentAppliedLinearAccClamp=<6000 6000 6000>
    componentLinearAccScale=<0.3 0.3 0.3>
    componentLinearVelScale=<0.4 0.4 0.4>
    overrideWorldGravity {z=-980}
    simSpaceSettings {maxAngularAcceleration=7, maxAngularVelocity=10, maxLinearAcceleration=150, maxLinearVelocity=300, worldAlpha=0.5}
   }
   n4 = "Constraint - dyn_zipper_upper_r_MSTR (PPPP)"!(
    ComponentPose=@n3
    Alpha=1
    ConstraintWeights_0=0.761
    ConstraintWeights_1:float=0.1189
    ConstraintWeights_2:float=0.0649
    ConstraintWeights_3:float=0.0552
   ) -> (Pose) {
    alphaBoolBlend {blendOption=Linear}
    boneToModify {boneName=dyn_zipper_upper_r_MSTR}
    constraintSetup [
     {targetBone {boneName=thigh_twist_01_r}}
     {targetBone {boneName=thigh_r}}
     {targetBone {boneName=deform_groin_r}}
     {targetBone {boneName=pelvis}}
    ]
    constraintWeights [1,1,1,1]
   }
   n5 = "Constraint - dyn_zipper_lower_r_MSTR (PPPP)"!(
    ComponentPose=@n4
    Alpha=1
    ConstraintWeights_0=0.6017
    ConstraintWeights_1:float=0.3693
    ConstraintWeights_2:float=0.0151
    ConstraintWeights_3:float=0.0139
   ) -> (Pose) {
    alphaBoolBlend {blendOption=Linear}
    boneToModify {boneName=dyn_zipper_lower_r_MSTR}
    constraintSetup [
     {targetBone {boneName=thigh_r}}
     {targetBone {boneName=thigh_twist_01_r}}
     {targetBone {boneName=calf_r}}
     {targetBone {boneName=deform_knee_out_r}}
    ]
    constraintWeights [1,1,1,1]
   }
   n6 = "Constraint - dyn_zipper_up_l_MSTR (PPPP)"!(
    ComponentPose=@n5
    Alpha=1
    ConstraintWeights_0=0.7607
    ConstraintWeights_1:float=0.1207
    ConstraintWeights_2:float=0.0642
    ConstraintWeights_3:float=0.0544
   ) -> (Pose) {
    alphaBoolBlend {blendOption=Linear}
    boneToModify {boneName=dyn_zipper_up_l_MSTR}
    constraintSetup [
     {targetBone {boneName=thigh_twist_01_l}}
     {targetBone {boneName=thigh_l}}
     {targetBone {boneName=deform_groin_l}}
     {targetBone {boneName=pelvis}}
    ]
    constraintWeights [1,1,1,1]
   }
   n7 = "Constraint - dyn_zipper_lower_l_MSTR (PPPP)"!(
    ComponentPose=@n6
    Alpha=1
    ConstraintWeights_0=0.6247
    ConstraintWeights_1:float=0.3415
    ConstraintWeights_2:float=0.0176
    ConstraintWeights_3:float=0.0163
   ) -> (Pose) {
    alphaBoolBlend {blendOption=Linear}
    boneToModify {boneName=dyn_zipper_lower_l_MSTR}
    constraintSetup [
     {targetBone {boneName=thigh_l}}
     {targetBone {boneName=thigh_twist_01_l}}
     {targetBone {boneName=calf_l}}
     {targetBone {boneName=deform_knee_out_l}}
    ]
    constraintWeights [1,1,1,1]
   }
   n8 = Anim Dynamics - Bone: dyn_zipper_upper_r!(ComponentPose=@n7, Alpha=1, GravityOverride:Vector="0,0,-980", ExternalForce="0,0,0") -> (Pose) {
    angularDampingOverride=0.99
    angularSpringConstant=500
    bAngularSpring=true
    bOverrideAngularDamping=true
    bUseGravityOverride=true
    gravityScale=0.35
    simulationSpace=RootRelative
    alphaBoolBlend {blendOption=Linear}
    boundBone {boneName=dyn_zipper_upper_r}
    componentAppliedLinearAccClamp=<5000 5000 5000>
    componentLinearAccScale=<0.5 0.5 0.5>
    componentLinearVelScale=<0.6 0.6 0.6>
    gravityOverride {z=-980}
    physicsBodyDefinitions [{
      boundBone {boneName=dyn_zipper_upper_r}
      boxExtents=<10 10 10>
      constraintSetup {angularLimitsMax=<4 0 20>, angularLimitsMin=<-4 -60 -20>, angularTarget=<1 0 0>}
      localJointOffset=<0.8 0 0>
     }]
   }
   n9 = Anim Dynamics - Bone: dyn_zipper_lower_r!(ComponentPose=@n8, Alpha=1, GravityOverride:Vector="0,0,-980", ExternalForce="0,0,0") -> (Pose) {
    angularDampingOverride=0.99
    angularSpringConstant=500
    bAngularSpring=true
    bOverrideAngularDamping=true
    bUseGravityOverride=true
    gravityScale=0.35
    simulationSpace=RootRelative
    alphaBoolBlend {blendOption=Linear}
    boundBone {boneName=dyn_zipper_lower_r}
    componentAppliedLinearAccClamp=<5000 5000 5000>
    componentLinearAccScale=<0.5 0.5 0.5>
    componentLinearVelScale=<0.6 0.6 0.6>
    gravityOverride {z=-980}
    physicsBodyDefinitions [{
      boundBone {boneName=dyn_zipper_lower_r}
      boxExtents=<11 11 11>
      constraintSetup {angularLimitsMax=<4 70 20>, angularLimitsMin=<-4 0 -20>, angularTarget=<1 0 0>}
      localJointOffset=<0.6 0 0>
     }]
   }
   n10 = Anim Dynamics - Bone: dyn_zipper_up_l!(ComponentPose=@n9, Alpha=1, GravityOverride:Vector="0,0,-980", ExternalForce="0,0,0") -> (Pose) {
    angularDampingOverride=0.99
    angularSpringConstant=500
    bAngularSpring=true
    bOverrideAngularDamping=true
    bUseGravityOverride=true
    gravityScale=0.35
    simulationSpace=RootRelative
    alphaBoolBlend {blendOption=Linear}
    boundBone {boneName=dyn_zipper_up_l}
    componentAppliedLinearAccClamp=<5000 5000 5000>
    componentLinearAccScale=<0.5 0.5 0.5>
    componentLinearVelScale=<0.6 0.6 0.6>
    gravityOverride {z=-980}
    physicsBodyDefinitions [{
      boundBone {boneName=dyn_zipper_up_l}
      boxExtents=<12 12 12>
      constraintSetup {angularLimitsMax=<4 0 20>, angularLimitsMin=<-4 -60 -20>, angularTarget=<1 0 0>}
      localJointOffset=<0.7 0 0>
     }]
   }
   n11 = Anim Dynamics - Bone: dyn_zipper_lower_l!(ComponentPose=@n10, Alpha=1, GravityOverride:Vector="0,0,-980", ExternalForce="0,0,0") -> (Pose) {
    angularDampingOverride=0.99
    angularSpringConstant=500
    bAngularSpring=true
    bOverrideAngularDamping=true
    bUseGravityOverride=true
    gravityScale=0.35
    simulationSpace=RootRelative
    alphaBoolBlend {blendOption=Linear}
    boundBone {boneName=dyn_zipper_lower_l}
    componentAppliedLinearAccClamp=<5000 5000 5000>
    componentLinearAccScale=<0.5 0.5 0.5>
    componentLinearVelScale=<0.6 0.6 0.6>
    gravityOverride {z=-980}
    physicsBodyDefinitions [{
      boundBone {boneName=dyn_zipper_lower_l}
      boxExtents=<12 12 12>
      constraintSetup {angularLimitsMax=<4 60 20>, angularLimitsMin=<-4 0 -20>, angularTarget=<1 0 0>}
      localJointOffset=<0.5 0 0>
     }]
   }
   n12 = Anim Dynamics - Bone: dyn_loop_l!(ComponentPose=@n11, Alpha=1, GravityOverride:Vector="0,0,0", ExternalForce="0,0,0") -> (Pose) {
    angularDampingOverride=0.95
    angularSpringConstant=4000
    bAngularSpring=true
    bOverrideAngularDamping=true
    bUseGravityOverride=true
    simulationSpace=RootRelative
    alphaBoolBlend {blendOption=Linear}
    boundBone {boneName=dyn_loop_l}
    componentAppliedLinearAccClamp=<5000 5000 5000>
    componentLinearAccScale=<0.4 0.4 0.4>
    componentLinearVelScale=<0.5 0.5 0.5>
    physicsBodyDefinitions [{
      boundBone {boneName=dyn_loop_l}
      boxExtents=<16 16 16>
      constraintSetup {angularLimitsMax=<0 30 0>, angularLimitsMin=<0 -20 0>, angularTarget=<1 0 0>}
      localJointOffset=<0.5 0 0>
     }]
   }
   n13 = Anim Dynamics - Bone: dyn_loop_r!(ComponentPose=@n12, Alpha=1, GravityOverride:Vector="0,0,0", ExternalForce="0,0,0") -> (Pose) {
    angularDampingOverride=0.95
    angularSpringConstant=4000
    bAngularSpring=true
    bOverrideAngularDamping=true
    bUseGravityOverride=true
    simulationSpace=RootRelative
    alphaBoolBlend {blendOption=Linear}
    boundBone {boneName=dyn_loop_r}
    componentAppliedLinearAccClamp=<5000 5000 5000>
    componentLinearAccScale=<0.4 0.4 0.4>
    componentLinearVelScale=<0.5 0.5 0.5>
    physicsBodyDefinitions [{
      boundBone {boneName=dyn_loop_r}
      boxExtents=<16 16 16>
      constraintSetup {angularLimitsMax=<0 30 0>, angularLimitsMin=<0 -20 0>, angularTarget=<1 0 0>}
      localJointOffset=<0.5 0 0>
     }]
   }
   n14 = Anim Dynamics - Bone: dyn_collar_01_r!(ComponentPose=@n13, Alpha=1, GravityOverride:Vector="0,0,0", ExternalForce="0,0,0") -> (Pose) {
    angularDampingOverride=0.999
    angularSpringConstant=4000
    bAngularSpring=true
    bOverrideAngularDamping=true
    bUseGravityOverride=true
    simulationSpace=RootRelative
    alphaBoolBlend {blendOption=Linear}
    boundBone {boneName=dyn_collar_01_r}
    componentAppliedLinearAccClamp=<5000 5000 5000>
    componentLinearAccScale=<0.6 0.6 0.6>
    componentLinearVelScale=<0.7 0.7 0.7>
    physicsBodyDefinitions [{
      boundBone {boneName=dyn_collar_01_r}
      boxExtents=<16 16 16>
      constraintSetup {angularLimitsMax=<0 30 0>, angularLimitsMin=<0 -30 0>, angularTarget=<1 0 0>}
      localJointOffset=<0.5 0 0>
     }]
   }
   n15 = Anim Dynamics - Bone: dyn_collar_01_l!(ComponentPose=@n14, Alpha=1, GravityOverride:Vector="0,0,0", ExternalForce="0,0,0") -> (Pose) {
    angularDampingOverride=0.999
    angularSpringConstant=4000
    bAngularSpring=true
    bOverrideAngularDamping=true
    bUseGravityOverride=true
    simulationSpace=RootRelative
    alphaBoolBlend {blendOption=Linear}
    boundBone {boneName=dyn_collar_01_l}
    componentAppliedLinearAccClamp=<5000 5000 5000>
    componentLinearAccScale=<0.6 0.6 0.6>
    componentLinearVelScale=<0.7 0.7 0.7>
    physicsBodyDefinitions [{
      boundBone {boneName=dyn_collar_01_l}
      boxExtents=<16 16 16>
      constraintSetup {angularLimitsMax=<0 30 0>, angularLimitsMin=<0 -30 0>, angularTarget=<1 0 0>}
      localJointOffset=<0.5 0 0>
     }]
   }
   Output Pose(Result=Component To Local(ComponentPose=@n15)) {layerGroup=None}
  ]}
 EventGraph {nodes [
   Event BlueprintUpdateAnimation()
   > n1 = Gravity Override State Function DEPRECATED(
    CharacterPartAnimInstance=Self-Reference()
    GameTimeInSeconds=GetGameTimeInSeconds()
    GravityOverrideParams=Get Gravity Override Params()
    RBAN Params=Get RBAN Params()
   )
   > Set GravityOverride(GravityOverride=@n1)
   > Set RBAN Settings Out(RBAN Settings Out:RBANSetting Struct Structure=@n1.RBAN SettingsOut) -> (Output_Get:RBANSetting Struct Structure)
   Is Valid(InputObject=Get Skel Mesh())
   > Gravity Override State Function DEPRECATED(CharacterPartAnimInstance=Self-Reference(), GameTimeInSeconds=GetGameTimeInSeconds(), GravityOverrideParams=Get Gravity Override Params())
   n2 = SelectVector(A="0.1,0.35,0.35", B="1,1,0.6", bPickA=OR Boolean(A=Get bIsSkyDiving(), B=Get bIsWaterJump() -> (bIsWaterJump:bool)))
   Set Component Linear Acc Scale(Component Linear Acc Scale:Vector=@n2) -> (Output_Get:Vector)
   > Set Component Linear Vel Scale(Component Linear Vel Scale:Vector=@n2) -> (Output_Get:Vector)
  ]}
]
variables [
 Component Linear Acc Scale {type=Vector}
 Component Linear Vel Scale {type=Vector}
 Gravity Override Params {type=GravityOverrideParamsStruct}
 GravityOverride {type=Vector}
 RBAN Params {
  type=RBANParamsStruct
  default {
   dBNO {
    componentLinearAccScale=<0.2 0.3 0.2>
    componentLinearVelScale=<0.3 0.4 0.3>
    simSpaceSettings {
     maxAngularAcceleration=1.5
     maxAngularVelocity=3
     maxLinearAcceleration=150
     maxLinearVelocity=300
     velocityScaleZ=1
     worldAlpha=0.75
     externalLinearDragV=<0 0 0>
    }
   }
   dBNO_Carried {
    componentLinearAccScale=<0.2 0.3 0.2>
    componentLinearVelScale=<0.3 0.4 0.3>
    simSpaceSettings {
     maxAngularAcceleration=1.5
     maxAngularVelocity=3
     maxLinearAcceleration=150
     maxLinearVelocity=300
     velocityScaleZ=1
     worldAlpha=0.75
     externalLinearDragV=<0 0 0>
    }
   }
   emoteOrMelee {
    componentLinearAccScale=<0.2 0.3 0.2>
    componentLinearVelScale=<0.3 0.4 0.3>
    simSpaceSettings {
     maxAngularAcceleration=1.5
     maxAngularVelocity=3
     maxLinearAcceleration=150
     maxLinearVelocity=300
     velocityScaleZ=1
     worldAlpha=0.75
     externalLinearDragV=<0 0 0>
    }
   }
   front_End {
    componentLinearAccScale=<0.2 0.3 0.2>
    componentLinearVelScale=<0.3 0.4 0.3>
    simSpaceSettings {
     maxAngularAcceleration=3
     maxAngularVelocity=5
     maxLinearAcceleration=150
     maxLinearVelocity=300
     velocityScaleZ=1
     worldAlpha=0.75
     externalLinearDragV=<0 0 0>
    }
   }
   onGround_Crouch_Moving {
    componentLinearAccScale=<0.2 0.3 0.2>
    componentLinearVelScale=<0.3 0.4 0.3>
    simSpaceSettings {
     maxAngularAcceleration=1.5
     maxAngularVelocity=3
     maxLinearAcceleration=150
     maxLinearVelocity=300
     velocityScaleZ=1
     worldAlpha=0.75
     externalLinearDragV=<0 0 0>
    }
   }
   onGround_Falling {simSpaceSettings {
     maxAngularAcceleration=1000
     maxAngularVelocity=3
     maxLinearAcceleration=300
     maxLinearVelocity=700
     velocityScaleZ=0
     externalLinearDragV=<0 0.5 0>
    }}
   onGround_Moving {
    componentLinearAccScale=<0.2 0.3 0.2>
    componentLinearVelScale=<0.3 0.4 0.3>
    simSpaceSettings {
     maxAngularAcceleration=1.5
     maxAngularVelocity=3
     maxLinearAcceleration=150
     maxLinearVelocity=300
     velocityScaleZ=1
     worldAlpha=0.75
     externalLinearDragV=<0 0 0>
    }
   }
   onGround_Standing {
    componentLinearAccScale=<0.3 0.3 0.5>
    componentLinearVelScale=<0.4 0.4 0.6>
    simSpaceSettings {
     maxAngularAcceleration=3
     maxAngularVelocity=5
     maxLinearAcceleration=150
     maxLinearVelocity=300
     velocityScaleZ=1
     worldAlpha=0.75
     externalLinearDragV=<0 0 0>
    }
   }
   rail_Grinding {
    componentLinearAccScale=<0.05 0.1 0.1>
    componentLinearVelScale=<0.1 0.2 0.2>
    simSpaceSettings {
     maxAngularAcceleration=1
     maxAngularVelocity=2
     maxLinearAcceleration=50
     maxLinearVelocity=150
     velocityScaleZ=1
     worldAlpha=0.5
     externalLinearDragV=<0 0 0>
    }
   }
   riding_Moving {
    componentLinearAccScale=<0.2 0.3 0.2>
    componentLinearVelScale=<0.3 0.4 0.3>
    simSpaceSettings {
     maxAngularAcceleration=1.5
     maxAngularVelocity=3
     maxLinearAcceleration=150
     maxLinearVelocity=300
     velocityScaleZ=1
     worldAlpha=0.75
     externalLinearDragV=<0 0 0>
    }
   }
   riding_Moving_Targeting {
    componentLinearAccScale=<0.2 0.3 0.2>
    componentLinearVelScale=<0.3 0.4 0.3>
    simSpaceSettings {
     maxAngularAcceleration=1.5
     maxAngularVelocity=3
     maxLinearAcceleration=150
     maxLinearVelocity=300
     velocityScaleZ=1
     worldAlpha=0.75
     externalLinearDragV=<0 0 0>
    }
   }
   riding_Standing {
    componentLinearAccScale=<0.2 0.3 0.2>
    componentLinearVelScale=<0.3 0.4 0.3>
    simSpaceSettings {
     maxAngularAcceleration=1.5
     maxAngularVelocity=3
     maxLinearAcceleration=150
     maxLinearVelocity=300
     velocityScaleZ=1
     worldAlpha=0.75
     externalLinearDragV=<0 0 0>
    }
   }
   skydive_DivingDown {
    componentLinearAccScale=<0.05 0.1 0.1>
    componentLinearVelScale=<0.1 0.2 0.2>
    simSpaceSettings {
     maxAngularAcceleration=1
     maxAngularVelocity=2
     maxLinearAcceleration=50
     maxLinearVelocity=150
     velocityScaleZ=1
     worldAlpha=0.5
     externalLinearDragV=<0 0 0>
    }
   }
   skydive_FloatingDown {
    componentLinearAccScale=<0.05 0.1 0.1>
    componentLinearVelScale=<0.1 0.2 0.2>
    simSpaceSettings {
     maxAngularAcceleration=1
     maxAngularVelocity=2
     maxLinearAcceleration=50
     maxLinearVelocity=150
     velocityScaleZ=1
     worldAlpha=0.5
     externalLinearDragV=<0 0 0>
    }
   }
   skydive_GoingUp {
    componentLinearAccScale=<0.05 0.1 0.1>
    componentLinearVelScale=<0.1 0.2 0.2>
    simSpaceSettings {
     maxAngularAcceleration=1
     maxAngularVelocity=2
     maxLinearAcceleration=50
     maxLinearVelocity=150
     velocityScaleZ=1
     worldAlpha=0.5
     externalLinearDragV=<0 0 0>
    }
   }
   skydive_Parachuting {
    componentLinearAccScale=<0.2 0.3 0.2>
    componentLinearVelScale=<0.3 0.4 0.3>
    simSpaceSettings {
     maxAngularAcceleration=1.5
     maxAngularVelocity=3
     maxLinearAcceleration=150
     maxLinearVelocity=300
     velocityScaleZ=1
     worldAlpha=0.75
     externalLinearDragV=<0 0 0>
    }
   }
   skydive_Surfing {
    componentLinearAccScale=<0.2 0.3 0.2>
    componentLinearVelScale=<0.3 0.4 0.3>
    simSpaceSettings {
     maxAngularAcceleration=1.5
     maxAngularVelocity=3
     maxLinearAcceleration=150
     maxLinearVelocity=300
     velocityScaleZ=1
     worldAlpha=0.75
     externalLinearDragV=<0 0 0>
    }
   }
   swimming {
    componentLinearAccScale=<0.2 0.3 0.2>
    componentLinearVelScale=<0.3 0.4 0.3>
    simSpaceSettings {
     maxAngularAcceleration=1.5
     maxAngularVelocity=3
     maxLinearAcceleration=150
     maxLinearVelocity=300
     velocityScaleZ=1
     worldAlpha=0.75
     externalLinearDragV=<0 0 0>
    }
   }
   tactical_Sprint {
    componentLinearAccScale=<0.05 0.1 0.1>
    componentLinearVelScale=<0.1 0.2 0.2>
    simSpaceSettings {
     maxAngularAcceleration=1
     maxAngularVelocity=2
     maxLinearAcceleration=50
     maxLinearVelocity=150
     velocityScaleZ=1
     worldAlpha=0.5
     externalLinearDragV=<0 0 0>
    }
   }
   ziplining {
    componentLinearAccScale=<0.05 0.1 0.1>
    componentLinearVelScale=<0.1 0.2 0.2>
    simSpaceSettings {
     maxAngularAcceleration=1
     maxAngularVelocity=2
     maxLinearAcceleration=50
     maxLinearVelocity=150
     velocityScaleZ=1
     worldAlpha=0.5
     externalLinearDragV=<0 0 0>
    }
   }
  }
 }
 RBAN Settings Out {
  type=RBANSettingStruct
  default {
   componentLinearAccScale=<0 0 0>
   componentLinearVelScale=<0 0 0>
   simSpaceSettings {
    dampingAlpha=1
    maxAngularAcceleration=1000
    maxAngularVelocity=3
    maxLinearAcceleration=300
    maxLinearVelocity=700
    velocityScaleZ=0
    worldAlpha=1
    externalAngularVelocity=<0 0 0>
    externalLinearDragV=<0 0.5 0>
    externalLinearVelocity=<0 0 0>
   }
  }
 }
 Skel Mesh {type=SkeletalMeshComponent}
]
