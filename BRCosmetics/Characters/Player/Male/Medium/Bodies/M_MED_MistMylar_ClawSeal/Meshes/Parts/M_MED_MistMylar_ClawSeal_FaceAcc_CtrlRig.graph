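# Control Rig for the face accessory part of the male medium MistMylar ClawSeal character, applying pose-driven corrections to hair bones as the head moves.
# A spherical pose reader measures the head bone's orientation relative to spine_05 and outputs a falloff weight.
# That weight blends transform changes onto master hair bones such as dyn_hair_21_MSTR and dyn_hair_18_MSTR, whose child dyn_hair bones inherit the adjustment.
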ControlRigBlueprint /BRCosmetics/Characters/Player/Male/Medium/Bodies/M_MED_MistMylar_ClawSeal/Meshes/Parts/M_MED_MistMylar_ClawSeal_FaceAcc_CtrlRig
graphs [
 RigVMFunctionLibrary {links [], nodes []}
 RigVMModel {nodes [
   RigUnit_BeginExecution = &RigUnit_BeginExecution()
   n1 = &RigUnit_SphericalPoseReader(
    ActiveRegionScaleFactors=(PositiveWidth=1,NegativeWidth=1,PositiveHeight=1,NegativeHeight=1)
    ActiveRegionSize=0.1
    Debug=(bDrawDebug=false,bDraw2D=False,bDrawLocalAxes=False,DebugScale=25,DebugSegments=20,DebugThickness=0.25)
    DriverAxis=(X=0,Y=1,Z=0)
    DriverItem=(Type=Bone,Name="head")
    FalloffRegionScaleFactors=(PositiveWidth=1,NegativeWidth=1,PositiveHeight=1,NegativeHeight=1)
    FalloffSize=0.2
    FlipHeightScaling=False
    FlipWidthScaling=False
    OptionalParentItem=(Type=Bone,Name="spine_05")
    RotationOffset=(X=90,Y=0,Z=-45)
    ExecutePin=@RigUnit_BeginExecution.ExecutePin
   )
   > &RigUnit_ModifyTransforms(
    ItemToModify="((Item=(Type=Bone,Name=\"dyn_hair_21_MSTR\"),Transform=(Rotation=(X=0,Y=-0.4664,Z=0,W=0.8846),Translation=(X=-0.1507,Y=0,Z=0.1032),Scale3D=(X=1,Y=1,Z=1))),(Item=(Type=Bone,Name=\"dyn_hair_18_MSTR\"),Transform=(Rotation=(X=0,Y=0,Z=0,W=1),Translation=(X=0,Y=0,Z= ...(truncated)"
    WeightMinimum=0
    Weight=@n1.OutputParam
   )
  ]}
]
hierarchy [
 root {loc=<0 0 0>, rot=<0 0 0 1>}
 pelvis {parent=root, loc=<0 -0.7071 93.1191>, rot=<0.0127 0.707 -0.0127 -0.707>}
 spine_01 {parent=pelvis, loc=<6.7494 -0.6388 0>, rot=<0 0 0.034 -0.9994>}
 spine_02 {parent=spine_01, loc=<7.252 -0.035 0>, rot=<0 0 0.0238 -0.9997>}
 spine_03 {parent=spine_02, loc=<6.0106 0 0>, rot=<0 0 0.0149 -0.9999>}
 spine_04 {parent=spine_03, loc=<8.1023 0 0>, rot=<0 0 -0.0132 -0.9999>}
 spine_05 {parent=spine_04, loc=<10.0142 0 0>, rot=<0 0 -0.0015 -1>}
 neck_01 {parent=spine_05, loc=<11.3138 0.1529 0>, rot=<0 0 -0.2003 -0.9797>}
 neck_02 {parent=neck_01, loc=<4.5627 0 0>, rot=<0 0 0.0504 -0.9987>}
 head {parent=neck_02, loc=<5.3914 0 0>, rot=<0 0 0.115 -0.9934>}
 dyn_hair_01 {parent=head, loc=<13.2506 5.2025 2.6763>, rot=<0.3641 -0.3793 0.162 -0.8351>}
 dyn_hair_01_A {parent=dyn_hair_01, loc=<0.6369 -1.7525 0.0582>, rot=<-0.0601 -0.1575 0.2098 -0.9631>}
 dyn_hair_01_B {parent=dyn_hair_01, loc=<1.569 1.596 -0.4247>, rot=<0.0274 -0.2645 -0.2115 -0.9405>}
 dyn_hair_02 {parent=head, loc=<11.9659 4.5653 5.155>, rot=<0.5058 0.018 0.8064 -0.3058>}
 dyn_hair_03 {parent=head, loc=<13.3728 6.5524 0.0555>, rot=<-0.0961 -0.8491 -0.1745 -0.4891>}
 dyn_hair_03_A {parent=dyn_hair_03, loc=<1.607 2.4011 -2.4829>, rot=<0.4681 -0.1277 -0.0242 -0.8741>}
 dyn_hair_03_B {parent=dyn_hair_03, loc=<3.3467 1.4371 -1.5675>, rot=<0.2548 -0.2456 0.0011 -0.9353>}
 dyn_hair_03_C {parent=dyn_hair_03, loc=<4.9951 -0.0631 -1.1574>, rot=<0.1043 -0.3297 -0.0892 -0.9341>}
 dyn_hair_03_D {parent=dyn_hair_03, loc=<5.5554 -2.5796 -0.8421>, rot=<0.0697 -0.2997 -0.0638 -0.9493>}
 dyn_hair_04 {parent=head, loc=<12.0678 -3.9715 4.6184>, rot=<0.4101 -0.2052 0.8459 0.2722>}
 dyn_hair_05 {parent=head, loc=<13.0509 2.6521 4.9455>, rot=<0.3568 -0.371 0.3431 -0.7857>}
 dyn_hair_06 {parent=head, loc=<11.1366 6.886 2.9768>, rot=<0.1114 0.5684 0.713 0.3953>}
 dyn_hair_07 {parent=head, loc=<10.5254 6.0417 4.7198>, rot=<0.189 0.3571 0.912 0.0714>}
 dyn_hair_08 {parent=head, loc=<5.9848 1.5961 -7.0057>, rot=<0.3943 -0.9139 -0.0755 -0.06>}
 dyn_hair_09 {parent=head, loc=<6.3063 -1.8029 -7.3364>, rot=<0.2265 -0.9355 0.2222 -0.1552>}
 dyn_hair_010 {parent=head, loc=<11.8968 3.2047 -5.1409>, rot=<0.4441 -0.7449 -0.0818 -0.4911>}
 dyn_hair_010_A {parent=dyn_hair_010, loc=<0.7945 -0.5214 -0.0723>, rot=<-0.042 -0.0665 -0.0278 -0.9965>}
 dyn_hair_010_B {parent=dyn_hair_010, loc=<2.9701 -0.6544 -1.251>, rot=<-0.3153 -0.2281 -0.1104 -0.9145>}
 dyn_hair_010_C {parent=dyn_hair_010, loc=<1.6672 1.4381 -0.8046>, rot=<0.122 -0.2155 0.0955 -0.9641>}
 dyn_hair_011 {parent=head, loc=<7.2356 2.7033 6.4855>, rot=<0.4013 -0.4425 0.767 -0.2342>}
 dyn_hair_011_A {parent=dyn_hair_011, loc=<2.1064 0.4703 1.2816>, rot=<-0.4643 -0.2193 0.0751 -0.8548>}
 dyn_hair_011_B {parent=dyn_hair_011, loc=<1.9362 0.2221 -0.0611>, rot=<0.0768 -0.0522 -0.098 -0.9908>}
 dyn_hair_011_C {parent=dyn_hair_011, loc=<2.1378 0.4099 -1.3278>, rot=<0.0386 0.3319 -0.1397 -0.9321>}
 dyn_hair_012 {parent=head, loc=<8.1525 -4.4304 6.3197>, rot=<0.1897 -0.145 0.9693 0.0593>}
 dyn_hair_013 {parent=head, loc=<1.4122 -2.66 -6.1777>, rot=<-0.0235 -0.9666 -0.1179 -0.2265>}
 dyn_hair_014 {parent=head, loc=<0.5119 -3.2949 -5.2214>, rot=<0.0458 -0.9576 0.1857 -0.2154>}
 dyn_hair_015 {parent=head, loc=<0.1544 -2.2166 5.8525>, rot=<0.2347 -0.1869 0.9386 0.1703>}
 dyn_hair_017 {parent=head, loc=<1.2074 -6.4162 0.9387>, rot=<0.18 -0.5942 0.77 0.1473>}
 dyn_hair_020 {parent=head, loc=<13.5481 -2.8094 -3.3401>, rot=<-0.1484 -0.7983 0.0718 -0.5793>}
 dyn_hair_21_MSTR {parent=head, loc=<-1.1305 -5.0578 -2.1792>, rot=<0.0716 -0.7764 0.6199 -0.0883>}
 dyn_hair_21 {parent=dyn_hair_21_MSTR, loc=<0 0 0>, rot=<0 0 0 1>}
 FACIAL_C_FacialRoot {parent=head, loc=<2.1524 -0.746 -0.0002>, rot=<0.4879 -0.5118 -0.4879 -0.5118>}
 FACIAL_C_Forehead {parent=FACIAL_C_FacialRoot, loc=<0.0147 -3.7445 8.9668>, rot=<-0.0661 0.0002 -0.0002 -0.9978>}
 FACIAL_L_CheekOuter {parent=FACIAL_C_FacialRoot, loc=<3.3362 0.1306 6.3853>, rot=<0.0344 -0.3789 0.0917 -0.9203>}
 FACIAL_R_CheekOuter {parent=FACIAL_C_FacialRoot, loc=<-3.3187 0.1353 6.3955>, rot=<0.0343 0.3793 -0.0919 -0.92>}
 FACIAL_L_CheekLower {parent=FACIAL_C_FacialRoot, loc=<3.5881 3.2953 5.8698>, rot=<0.0142 -0.4279 0.0479 -0.9025>}
 FACIAL_R_CheekLower {parent=FACIAL_C_FacialRoot, loc=<-3.5838 3.3109 5.9441>, rot=<0.0141 0.4283 -0.0482 -0.9022>}
 FACIAL_L_Masseter {parent=FACIAL_C_FacialRoot, loc=<5.5567 0.9377 4.3935>, rot=<0.0469 0.1129 0.0103 -0.9924>}
 FACIAL_R_Masseter {parent=FACIAL_C_FacialRoot, loc=<-5.5434 0.9374 4.3889>, rot=<0.0469 -0.1124 -0.0106 -0.9925>}
 FACIAL_L_Ear {parent=FACIAL_C_FacialRoot, loc=<6.2746 -0.5417 0.914>, rot=<-0.0489 0.0001 -0.0263 -0.9985>}
 FACIAL_R_Ear {parent=FACIAL_C_FacialRoot, loc=<-6.2657 -0.539 0.9094>, rot=<-0.049 0.0003 0.026 -0.9985>}
 dyn_hair_18_MSTR {parent=head, loc=<0.2482 -5.9073 0.9292>, rot=<0.0026 -0.6293 0.7771 -0.0032>}
 dyn_hair_018 {parent=dyn_hair_18_MSTR, loc=<0 0 0>, rot=<0 0 0 1>}
 dyn_hair_19_MSTR {parent=head, loc=<0.1106 -4.4577 -3.8303>, rot=<-0.0926 -0.7838 0.609 0.0783>}
 dyn_hair_019 {parent=dyn_hair_19_MSTR, loc=<0 0 0>, rot=<0 0 0 1>}
 dyn_hair_16_MSTR {parent=head, loc=<0.2211 -4.5612 4.4775>, rot=<0.2721 -0.2514 0.9078 0.1967>}
 dyn_hair_016 {parent=dyn_hair_16_MSTR, loc=<0 0 0>, rot=<0 0 0 1>}
]
