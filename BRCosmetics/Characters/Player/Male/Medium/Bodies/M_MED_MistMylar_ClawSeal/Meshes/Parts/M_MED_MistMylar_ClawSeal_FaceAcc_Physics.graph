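# Physics asset for the face accessory part of the male medium MistMylar ClawSeal character, defining bodies for its dynamic hair bones.
# One large capsule covers the head, and five small capsules cover individual dyn_hair bones.
# Each hair body is constrained directly to the head rather than chained, with moderate swing limits and very tight twist.
# The dyn_hair_010 constraint is narrower than the others on its first swing axis.
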PhysicsAsset /BRCosmetics/Characters/Player/Male/Medium/Bodies/M_MED_MistMylar_ClawSeal/Meshes/Parts/M_MED_MistMylar_ClawSeal_FaceAcc_Physics
bodies {
 dyn_hair_01 [{
   shapeType=Capsule
   length=3.2344
   radius=1.6927
   shapeName=dyn_hair_01_capsule
   center=<2.1016 0 1.6176>
   rotation=<rot 71.6014 -0.0002 -0.0002>
  }]
 dyn_hair_010 [{
   shapeType=Capsule
   length=1.2804
   radius=1.0755
   shapeName=dyn_hair_010_capsule
   center=<2.0287 0 0.8488>
   rotation=<rot 63.936 -0.0001 -0.0001>
  }]
 dyn_hair_03 [{shapeType=Capsule, length=2.7581, radius=1.8724, shapeName=dyn_hair_03_capsule, center=<2.0366 0.721 1.5959>, rotation=<rot 59.108 90 90>}]
 dyn_hair_06 [{
   shapeType=Capsule
   length=2.9504
   radius=0.9669
   shapeName=dyn_hair_06_capsule
   center=<2.4284 -0.4159 1.7366>
   rotation=<rot 0 -91.6887 -84.8033>
  }]
 dyn_hair_07 [{
   shapeType=Capsule
   length=2.1886
   radius=0.6468
   shapeName=dyn_hair_07_capsule
   center=<2.5329 0.0483 1.3903>
   rotation=<rot 0 -88.5481 -75.7563>
  }]
 head [{shapeType=Capsule, length=6.4253, radius=7.9821, shapeName=head_capsule, center=<3.648 2.195 0>, rotation=<rot 90 0 0>}]
}
constraints [
 {bone1Name=dyn_hair_01, bone2Name=head, swing1LimitDegrees=30, swing2LimitDegrees=30, twistLimitDegrees=2}
 {bone1Name=dyn_hair_03, bone2Name=head, swing1LimitDegrees=30, swing2LimitDegrees=30, twistLimitDegrees=2}
 {bone1Name=dyn_hair_010, bone2Name=head, swing1LimitDegrees=10, swing2LimitDegrees=30, twistLimitDegrees=2}
 {bone1Name=dyn_hair_06, bone2Name=head, swing1LimitDegrees=30, swing2LimitDegrees=30, twistLimitDegrees=2}
 {bone1Name=dyn_hair_07, bone2Name=head, swing1LimitDegrees=30, swing2LimitDegrees=30, twistLimitDegrees=2}
]
