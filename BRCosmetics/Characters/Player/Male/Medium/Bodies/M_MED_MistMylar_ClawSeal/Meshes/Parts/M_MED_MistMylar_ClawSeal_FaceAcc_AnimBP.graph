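# Post-process-style AnimBP that adds secondary hair motion to the MistMylar ClawSeal face accessory.
# It copies the attached parent's pose through a Control Rig, then runs a head-relative RigidBody simulation.
# A serial chain of sprung, angle-limited AnimDynamics nodes follows, one per dyn_hair bone, with stiffer springs on later bones.
# Each update, a deprecated state function sets Gravity Override, though the physics nodes use zero-gravity literals.
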AnimBlueprint /BRCosmetics/Characters/Player/Male/Medium/Bodies/M_MED_MistMylar_ClawSeal/Meshes/Parts/M_MED_MistMylar_ClawSeal_FaceAcc_AnimBP
target_skeleton=&/M_MED_MistMylar_ClawSeal_FaceAcc_Skeleton
graphs [
 AnimGraph {nodes [
   n1 = Control Rig(Source=Copy Pose From Mesh() {bUseAttachedParent=true}) {
    alphaBoolBlend {blendOption=Linear}
    controlRigAssetReference {blueprintRigClass=&/M_MED_MistMylar_ClawSeal_FaceAcc_CtrlRig.M_MED_MistMylar_ClawSeal_FaceAcc_CtrlRig_C}
   }
   n2 = RigidBody!(ComponentPose=Local To Component(LocalPose=@n1), Alpha, OverrideWorldGravity:Vector="0,0,0", ExternalForce) -> (Pose) {
    bClampLinearTranslationLimitToRefPose=true
    bOverrideWorldGravity=true
    overridePhysicsAsset=&/M_MED_MistMylar_ClawSeal_FaceAcc_Physics
    simulationSpace=BaseBoneSpace
    alphaBoolBlend {blendOption=Linear}
    baseBoneRef {boneName=head}
    componentLinearAccScale=<0.3 0.3 0.3>
    componentLinearVelScale=<0.4 0.4 0.4>
    simSpaceSettings {maxAngularAcceleration=5, maxAngularVelocity=10, maxLinearAcceleration=150, maxLinearVelocity=300, worldAlpha=0.5}
   }
   n3 = Anim Dynamics - Bone: dyn_hair_01_A!(ComponentPose=@n2, Alpha=1, GravityOverride:Vector="0,0,0", ExternalForce="0,0,0") -> (Pose) {
    angularDampingOverride=0.95
    angularSpringConstant=1500
    bAngularSpring=true
    bOverrideAngularDamping=true
    bUseGravityOverride=true
    simulationSpace=RootRelative
    alphaBoolBlend {blendOption=Linear}
    boundBone {boneName=dyn_hair_01_A}
    componentAppliedLinearAccClamp=<5000 5000 5000>
    componentLinearAccScale=<0.6 0.6 0.6>
    componentLinearVelScale=<0.7 0.7 0.7>
    physicsBodyDefinitions [{
      boundBone {boneName=dyn_hair_01_A}
      boxExtents=<10 10 10>
      constraintSetup {angularLimitsMax=<0 0 4>, angularLimitsMin=<0 -10 -4>, angularTarget=<1 0 0>}
      localJointOffset=<0.7 0 0>
     }]
   }
   n4 = Anim Dynamics - Bone: dyn_hair_01_B!(ComponentPose=@n3, Alpha=1, GravityOverride:Vector="0,0,0", ExternalForce="0,0,0") -> (Pose) {
    angularDampingOverride=0.95
    angularSpringConstant=1500
    bAngularSpring=true
    bOverrideAngularDamping=true
    bUseGravityOverride=true
    simulationSpace=RootRelative
    alphaBoolBlend {blendOption=Linear}
    boundBone {boneName=dyn_hair_01_B}
    componentAppliedLinearAccClamp=<5000 5000 5000>
    componentLinearAccScale=<0.6 0.6 0.6>
    componentLinearVelScale=<0.7 0.7 0.7>
    physicsBodyDefinitions [{
      boundBone {boneName=dyn_hair_01_B}
      boxExtents=<12 12 12>
      constraintSetup {angularLimitsMax=<0 0 0>, angularLimitsMin=<0 -10 0>, angularTarget=<1 0 0>}
      localJointOffset=<0.6 0 0>
     }]
   }
   n5 = Anim Dynamics - Bone: dyn_hair_02!(ComponentPose=@n4, Alpha=1, GravityOverride:Vector="0,0,0", ExternalForce) -> (Pose) {
    angularDampingOverride=0.95
    angularSpringConstant=1500
    bAngularSpring=true
    bOverrideAngularDamping=true
    bUseGravityOverride=true
    simulationSpace=RootRelative
    alphaBoolBlend {blendOption=Linear}
    boundBone {boneName=dyn_hair_02}
    componentAppliedLinearAccClamp=<5000 5000 5000>
    componentLinearAccScale=<0.6 0.6 0.6>
    componentLinearVelScale=<0.7 0.7 0.7>
    physicsBodyDefinitions [{
      boundBone {boneName=dyn_hair_02}
      boxExtents=<14 14 14>
      constraintSetup {angularLimitsMax=<0 0 4>, angularLimitsMin=<0 -10 -4>, angularTarget=<1 0 0>}
      localJointOffset=<0.7 0 0>
     }]
   }
   n6 = Anim Dynamics - Bone: dyn_hair_03_A!(ComponentPose=@n5, Alpha=1, GravityOverride:Vector="0,0,0", ExternalForce="0,0,0") -> (Pose) {
    angularDampingOverride=0.95
    angularSpringConstant=1500
    bAngularSpring=true
    bOverrideAngularDamping=true
    bUseGravityOverride=true
    simulationSpace=RootRelative
    alphaBoolBlend {blendOption=Linear}
    boundBone {boneName=dyn_hair_03_A}
    componentAppliedLinearAccClamp=<5000 5000 5000>
    componentLinearAccScale=<0.6 0.6 0.6>
    componentLinearVelScale=<0.7 0.7 0.7>
    physicsBodyDefinitions [{
      boundBone {boneName=dyn_hair_03_A}
      boxExtents=<12 12 12>
      constraintSetup {angularLimitsMax=<0 0 4>, angularLimitsMin=<0 -10 -4>, angularTarget=<1 0 0>}
      localJointOffset=<0.8 0 0>
     }]
   }
   n7 = Anim Dynamics - Bone: dyn_hair_03_B!(ComponentPose=@n6, Alpha=1, GravityOverride:Vector="0,0,0", ExternalForce="0,0,0") -> (Pose) {
    angularDampingOverride=0.95
    angularSpringConstant=1500
    bAngularSpring=true
    bOverrideAngularDamping=true
    bUseGravityOverride=true
    simulationSpace=RootRelative
    alphaBoolBlend {blendOption=Linear}
    boundBone {boneName=dyn_hair_03_B}
    componentAppliedLinearAccClamp=<5000 5000 5000>
    componentLinearAccScale=<0.6 0.6 0.6>
    componentLinearVelScale=<0.7 0.7 0.7>
    physicsBodyDefinitions [{
      boundBone {boneName=dyn_hair_03_B}
      boxExtents=<10 10 10>
      constraintSetup {angularLimitsMax=<0 0 4>, angularLimitsMin=<0 -10 -4>, angularTarget=<1 0 0>}
      localJointOffset=<0.6 0 0>
     }]
   }
   n8 = Anim Dynamics - Bone: dyn_hair_03_C!(ComponentPose=@n7, Alpha=1, GravityOverride:Vector="0,0,0", ExternalForce="0,0,0") -> (Pose) {
    angularDampingOverride=0.95
    angularSpringConstant=1500
    bAngularSpring=true
    bOverrideAngularDamping=true
    bUseGravityOverride=true
    simulationSpace=RootRelative
    alphaBoolBlend {blendOption=Linear}
    boundBone {boneName=dyn_hair_03_C}
    componentAppliedLinearAccClamp=<5000 5000 5000>
    componentLinearAccScale=<0.6 0.6 0.6>
    componentLinearVelScale=<0.7 0.7 0.7>
    physicsBodyDefinitions [{
      boundBone {boneName=dyn_hair_03_C}
      boxExtents=<10 10 10>
      constraintSetup {angularLimitsMax=<0 0 4>, angularLimitsMin=<0 -10 -4>, angularTarget=<1 0 0>}
      localJointOffset=<0.7 0 0>
     }]
   }
   n9 = Anim Dynamics - Bone: dyn_hair_03_D!(ComponentPose=@n8, Alpha=1, GravityOverride:Vector="0,0,0", ExternalForce="0,0,0") -> (Pose) {
    angularDampingOverride=0.95
    angularSpringConstant=1500
    bAngularSpring=true
    bOverrideAngularDamping=true
    bUseGravityOverride=true
    simulationSpace=RootRelative
    alphaBoolBlend {blendOption=Linear}
    boundBone {boneName=dyn_hair_03_D}
    componentAppliedLinearAccClamp=<5000 5000 5000>
    componentLinearAccScale=<0.6 0.6 0.6>
    componentLinearVelScale=<0.7 0.7 0.7>
    physicsBodyDefinitions [{
      boundBone {boneName=dyn_hair_03_D}
      boxExtents=<14 14 14>
      constraintSetup {angularLimitsMax=<0 0 4>, angularLimitsMin=<0 -10 -4>, angularTarget=<1 0 0>}
      localJointOffset=<0.6 0 0>
     }]
   }
   n10 = Anim Dynamics - Bone: dyn_hair_04!(ComponentPose=@n9, Alpha=1, GravityOverride:Vector="0,0,0", ExternalForce="0,0,0") -> (Pose) {
    angularDampingOverride=0.95
    angularSpringConstant=1500
    bAngularSpring=true
    bOverrideAngularDamping=true
    bUseGravityOverride=true
    simulationSpace=RootRelative
    alphaBoolBlend {blendOption=Linear}
    boundBone {boneName=dyn_hair_04}
    componentAppliedLinearAccClamp=<5000 5000 5000>
    componentLinearAccScale=<0.6 0.6 0.6>
    componentLinearVelScale=<0.7 0.7 0.7>
    physicsBodyDefinitions [{
      boundBone {boneName=dyn_hair_04}
      boxExtents=<12 12 12>
      constraintSetup {angularLimitsMax=<0 0 4>, angularLimitsMin=<0 -10 -4>, angularTarget=<1 0 0>}
      localJointOffset=<0.8 0 0>
     }]
   }
   n11 = Anim Dynamics - Bone: dyn_hair_05!(ComponentPose=@n10, Alpha=1, GravityOverride:Vector="0,0,0", ExternalForce="0,0,0") -> (Pose) {
    angularDampingOverride=0.95
    angularSpringConstant=1500
    bAngularSpring=true
    bOverrideAngularDamping=true
    bUseGravityOverride=true
    simulationSpace=RootRelative
    alphaBoolBlend {blendOption=Linear}
    boundBone {boneName=dyn_hair_05}
    componentAppliedLinearAccClamp=<5000 5000 5000>
    componentLinearAccScale=<0.6 0.6 0.6>
    componentLinearVelScale=<0.7 0.7 0.7>
    physicsBodyDefinitions [{
      boundBone {boneName=dyn_hair_05}
      boxExtents=<10 10 10>
      constraintSetup {angularLimitsMax=<0 0 4>, angularLimitsMin=<0 -10 -4>, angularTarget=<1 0 0>}
      localJointOffset=<0.7 0 0>
     }]
   }
   n12 = Anim Dynamics - Bone: dyn_hair_06!(ComponentPose=@n11, Alpha=1, GravityOverride:Vector="0,0,0", ExternalForce="0,0,0") -> (Pose) {
    angularDampingOverride=0.95
    angularSpringConstant=1500
    bAngularSpring=true
    bOverrideAngularDamping=true
    bUseGravityOverride=true
    simulationSpace=RootRelative
    alphaBoolBlend {blendOption=Linear}
    boundBone {boneName=dyn_hair_06}
    componentAppliedLinearAccClamp=<5000 5000 5000>
    componentLinearAccScale=<0.6 0.6 0.6>
    componentLinearVelScale=<0.7 0.7 0.7>
    physicsBodyDefinitions [{
      boundBone {boneName=dyn_hair_06}
      boxExtents=<16 16 16>
      constraintSetup {angularLimitsMax=<0 0 4>, angularLimitsMin=<0 -10 -4>, angularTarget=<1 0 0>}
      localJointOffset=<0.7 0 0>
     }]
   }
   n13 = Anim Dynamics - Bone: dyn_hair_07!(ComponentPose=@n12, Alpha=1, GravityOverride:Vector="0,0,0", ExternalForce="0,0,0") -> (Pose) {
    angularDampingOverride=0.95
    angularSpringConstant=1500
    bAngularSpring=true
    bOverrideAngularDamping=true
    bUseGravityOverride=true
    simulationSpace=RootRelative
    alphaBoolBlend {blendOption=Linear}
    boundBone {boneName=dyn_hair_07}
    componentAppliedLinearAccClamp=<5000 5000 5000>
    componentLinearAccScale=<0.6 0.6 0.6>
    componentLinearVelScale=<0.7 0.7 0.7>
    physicsBodyDefinitions [{
      boundBone {boneName=dyn_hair_07}
      boxExtents=<16 16 16>
      constraintSetup {angularLimitsMax=<0 0 4>, angularLimitsMin=<0 -10 -4>, angularTarget=<1 0 0>}
      localJointOffset=<0.7 0 0>
     }]
   }
   n14 = Anim Dynamics - Bone: dyn_hair_08!(ComponentPose=@n13, Alpha=1, GravityOverride:Vector="0,0,0", ExternalForce="0,0,0") -> (Pose) {
    angularDampingOverride=0.95
    angularSpringConstant=1500
    bAngularSpring=true
    bOverrideAngularDamping=true
    bUseGravityOverride=true
    simulationSpace=RootRelative
    alphaBoolBlend {blendOption=Linear}
    boundBone {boneName=dyn_hair_08}
    componentAppliedLinearAccClamp=<5000 5000 5000>
    componentLinearAccScale=<0.6 0.6 0.6>
    componentLinearVelScale=<0.7 0.7 0.7>
    physicsBodyDefinitions [{
      boundBone {boneName=dyn_hair_08}
      boxExtents=<10 10 10>
      constraintSetup {angularLimitsMax=<0 0 4>, angularLimitsMin=<0 -10 -4>, angularTarget=<1 0 0>}
      localJointOffset=<0.6 0 0>
     }]
   }
   n15 = Anim Dynamics - Bone: dyn_hair_09!(ComponentPose=@n14, Alpha=1, GravityOverride:Vector="0,0,0", ExternalForce="0,0,0") -> (Pose) {
    angularDampingOverride=0.95
    angularSpringConstant=1500
    bAngularSpring=true
    bOverrideAngularDamping=true
    bUseGravityOverride=true
    simulationSpace=RootRelative
    alphaBoolBlend {blendOption=Linear}
    boundBone {boneName=dyn_hair_09}
    componentAppliedLinearAccClamp=<5000 5000 5000>
    componentLinearAccScale=<0.6 0.6 0.6>
    componentLinearVelScale=<0.7 0.7 0.7>
    physicsBodyDefinitions [{
      boundBone {boneName=dyn_hair_09}
      boxExtents=<10 10 10>
      constraintSetup {angularLimitsMax=<0 0 4>, angularLimitsMin=<0 -10 -4>, angularTarget=<1 0 0>}
      localJointOffset=<0.6 0 0>
     }]
   }
   n16 = Anim Dynamics - Bone: dyn_hair_010_A!(ComponentPose=@n15, Alpha=1, GravityOverride:Vector="0,0,0", ExternalForce="0,0,0") -> (Pose) {
    angularDampingOverride=0.95
    angularSpringConstant=1500
    bAngularSpring=true
    bOverrideAngularDamping=true
    bUseGravityOverride=true
    simulationSpace=RootRelative
    alphaBoolBlend {blendOption=Linear}
    boundBone {boneName=dyn_hair_010_A}
    componentAppliedLinearAccClamp=<5000 5000 5000>
    componentLinearAccScale=<0.6 0.6 0.6>
    componentLinearVelScale=<0.7 0.7 0.7>
    physicsBodyDefinitions [{
      boundBone {boneName=dyn_hair_010_A}
      boxExtents=<10 10 10>
      constraintSetup {angularLimitsMax=<0 0 4>, angularLimitsMin=<0 -10 -4>, angularTarget=<1 0 0>}
      localJointOffset=<0.7 0 0>
     }]
   }
   n17 = Anim Dynamics - Bone: dyn_hair_010_B!(ComponentPose=@n16, Alpha=1, GravityOverride:Vector="0,0,0", ExternalForce="0,0,0") -> (Pose) {
    angularDampingOverride=0.95
    angularSpringConstant=1500
    bAngularSpring=true
    bOverrideAngularDamping=true
    bUseGravityOverride=true
    simulationSpace=RootRelative
    alphaBoolBlend {blendOption=Linear}
    boundBone {boneName=dyn_hair_010_B}
    componentAppliedLinearAccClamp=<5000 5000 5000>
    componentLinearAccScale=<0.6 0.6 0.6>
    componentLinearVelScale=<0.7 0.7 0.7>
    physicsBodyDefinitions [{
      boundBone {boneName=dyn_hair_010_B}
      boxExtents=<10 10 10>
      constraintSetup {angularLimitsMax=<0 0 4>, angularLimitsMin=<0 -10 -4>, angularTarget=<1 0 0>}
      localJointOffset=<0.6 0 0>
     }]
   }
   n18 = Anim Dynamics - Bone: dyn_hair_010_C!(ComponentPose=@n17, Alpha=1, GravityOverride:Vector="0,0,0", ExternalForce="0,0,0") -> (Pose) {
    angularDampingOverride=0.95
    angularSpringConstant=1500
    bAngularSpring=true
    bOverrideAngularDamping=true
    bUseGravityOverride=true
    simulationSpace=RootRelative
    alphaBoolBlend {blendOption=Linear}
    boundBone {boneName=dyn_hair_010_C}
    componentAppliedLinearAccClamp=<5000 5000 5000>
    componentLinearAccScale=<0.6 0.6 0.6>
    componentLinearVelScale=<0.7 0.7 0.7>
    physicsBodyDefinitions [{
      boundBone {boneName=dyn_hair_010_C}
      boxExtents=<10 10 10>
      constraintSetup {angularLimitsMax=<0 0 4>, angularLimitsMin=<0 -10 -4>, angularTarget=<1 0 0>}
      localJointOffset=<0.8 0 0>
     }]
   }
   n19 = Anim Dynamics - Bone: dyn_hair_012!(ComponentPose=@n18, Alpha=1, GravityOverride:Vector="0,0,0", ExternalForce="0,0,0") -> (Pose) {
    angularDampingOverride=0.995
    angularSpringConstant=2000
    bAngularSpring=true
    bOverrideAngularDamping=true
    bUseGravityOverride=true
    simulationSpace=RootRelative
    alphaBoolBlend {blendOption=Linear}
    boundBone {boneName=dyn_hair_012}
    componentAppliedLinearAccClamp=<5000 5000 5000>
    componentLinearAccScale=<0.6 0.6 0.6>
    componentLinearVelScale=<0.7 0.7 0.7>
    physicsBodyDefinitions [{
      boundBone {boneName=dyn_hair_012}
      boxExtents=<16 16 16>
      constraintSetup {angularLimitsMax=<0 0 8>, angularLimitsMin=<0 -10 -8>, angularTarget=<1 0 0>}
      localJointOffset=<0.5 0 0>
     }]
   }
   n20 = Anim Dynamics - Bone: dyn_hair_013!(ComponentPose=@n19, Alpha=1, GravityOverride:Vector="0,0,0", ExternalForce="0,0,0") -> (Pose) {
    angularDampingOverride=0.995
    angularSpringConstant=2000
    bAngularSpring=true
    bOverrideAngularDamping=true
    bUseGravityOverride=true
    simulationSpace=RootRelative
    alphaBoolBlend {blendOption=Linear}
    boundBone {boneName=dyn_hair_013}
    componentAppliedLinearAccClamp=<5000 5000 5000>
    componentLinearAccScale=<0.6 0.6 0.6>
    componentLinearVelScale=<0.7 0.7 0.7>
    physicsBodyDefinitions [{
      boundBone {boneName=dyn_hair_013}
      boxExtents=<16 16 16>
      constraintSetup {angularLimitsMax=<0 0 8>, angularLimitsMin=<0 -20 -8>, angularTarget=<1 0 0>}
      localJointOffset=<0.5 0 0>
     }]
   }
   n21 = Anim Dynamics - Bone: dyn_hair_014!(ComponentPose=@n20, Alpha=1, GravityOverride:Vector="0,0,0", ExternalForce="0,0,0") -> (Pose) {
    angularDampingOverride=0.995
    angularSpringConstant=2000
    bAngularSpring=true
    bOverrideAngularDamping=true
    bUseGravityOverride=true
    simulationSpace=RootRelative
    alphaBoolBlend {blendOption=Linear}
    boundBone {boneName=dyn_hair_014}
    componentAppliedLinearAccClamp=<5000 5000 5000>
    componentLinearAccScale=<0.6 0.6 0.6>
    componentLinearVelScale=<0.7 0.7 0.7>
    physicsBodyDefinitions [{
      boundBone {boneName=dyn_hair_014}
      boxExtents=<12 12 12>
      constraintSetup {angularLimitsMax=<0 0 8>, angularLimitsMin=<0 -20 -8>, angularTarget=<1 0 0>}
      localJointOffset=<0.7 0 0>
     }]
   }
   n22 = Anim Dynamics - Bone: dyn_hair_015!(ComponentPose=@n21, Alpha=1, GravityOverride:Vector="0,0,0", ExternalForce="0,0,0") -> (Pose) {
    angularDampingOverride=0.995
    angularSpringConstant=2000
    bAngularSpring=true
    bOverrideAngularDamping=true
    bUseGravityOverride=true
    simulationSpace=RootRelative
    alphaBoolBlend {blendOption=Linear}
    boundBone {boneName=dyn_hair_015}
    componentAppliedLinearAccClamp=<5000 5000 5000>
    componentLinearAccScale=<0.6 0.6 0.6>
    componentLinearVelScale=<0.7 0.7 0.7>
    physicsBodyDefinitions [{
      boundBone {boneName=dyn_hair_015}
      boxExtents=<16 16 16>
      constraintSetup {angularLimitsMax=<0 0 8>, angularLimitsMin=<0 -20 -8>, angularTarget=<1 0 0>}
      localJointOffset=<0.5 0 0>
     }]
   }
   n23 = Anim Dynamics - Bone: dyn_hair_016!(ComponentPose=@n22, Alpha=1, GravityOverride:Vector="0,0,0", ExternalForce="0,0,0") -> (Pose) {
    angularDampingOverride=0.995
    angularSpringConstant=2000
    bAngularSpring=true
    bOverrideAngularDamping=true
    bUseGravityOverride=true
    simulationSpace=RootRelative
    alphaBoolBlend {blendOption=Linear}
    boundBone {boneName=dyn_hair_016}
    componentAppliedLinearAccClamp=<5000 5000 5000>
    componentLinearAccScale=<0.6 0.6 0.6>
    componentLinearVelScale=<0.7 0.7 0.7>
    physicsBodyDefinitions [{
      boundBone {boneName=dyn_hair_016}
      boxExtents=<14 14 14>
      constraintSetup {angularLimitsMax=<0 0 8>, angularLimitsMin=<0 -20 -8>, angularTarget=<1 0 0>}
      localJointOffset=<0.6 0 0>
     }]
   }
   n24 = Anim Dynamics - Bone: dyn_hair_017!(ComponentPose=@n23, Alpha=1, GravityOverride:Vector="0,0,0", ExternalForce="0,0,0") -> (Pose) {
    angularDampingOverride=0.995
    angularSpringConstant=2000
    bAngularSpring=true
    bOverrideAngularDamping=true
    bUseGravityOverride=true
    simulationSpace=RootRelative
    alphaBoolBlend {blendOption=Linear}
    boundBone {boneName=dyn_hair_017}
    componentAppliedLinearAccClamp=<5000 5000 5000>
    componentLinearAccScale=<0.6 0.6 0.6>
    componentLinearVelScale=<0.7 0.7 0.7>
    physicsBodyDefinitions [{
      boundBone {boneName=dyn_hair_017}
      boxExtents=<14 14 14>
      constraintSetup {angularLimitsMax=<0 0 8>, angularLimitsMin=<0 -20 -8>, angularTarget=<1 0 0>}
      localJointOffset=<0.6 0 0>
     }]
   }
   n25 = Anim Dynamics - Bone: dyn_hair_018!(ComponentPose=@n24, Alpha=1, GravityOverride:Vector="0,0,0", ExternalForce="0,0,0") -> (Pose) {
    angularDampingOverride=0.995
    angularSpringConstant=2000
    bAngularSpring=true
    bOverrideAngularDamping=true
    bUseGravityOverride=true
    simulationSpace=RootRelative
    alphaBoolBlend {blendOption=Linear}
    boundBone {boneName=dyn_hair_018}
    componentAppliedLinearAccClamp=<5000 5000 5000>
    componentLinearAccScale=<0.6 0.6 0.6>
    componentLinearVelScale=<0.7 0.7 0.7>
    physicsBodyDefinitions [{
      boundBone {boneName=dyn_hair_018}
      boxExtents=<12 12 12>
      constraintSetup {angularLimitsMax=<0 0 12>, angularLimitsMin=<0 -30 -12>, angularTarget=<1 0 0>}
      localJointOffset=<0.7 0 0>
     }]
   }
   n26 = Anim Dynamics - Bone: dyn_hair_019!(ComponentPose=@n25, Alpha=1, GravityOverride:Vector="0,0,0", ExternalForce="0,0,0") -> (Pose) {
    angularDampingOverride=0.995
    angularSpringConstant=2000
    bAngularSpring=true
    bOverrideAngularDamping=true
    bUseGravityOverride=true
    simulationSpace=RootRelative
    alphaBoolBlend {blendOption=Linear}
    boundBone {boneName=dyn_hair_019}
    componentAppliedLinearAccClamp=<5000 5000 5000>
    componentLinearAccScale=<0.6 0.6 0.6>
    componentLinearVelScale=<0.7 0.7 0.7>
    physicsBodyDefinitions [{
      boundBone {boneName=dyn_hair_019}
      boxExtents=<16 16 16>
      constraintSetup {angularLimitsMax=<0 0 8>, angularLimitsMin=<0 -30 -8>, angularTarget=<1 0 0>}
      localJointOffset=<0.5 0 0>
     }]
   }
   n27 = Anim Dynamics - Bone: dyn_hair_020!(ComponentPose=@n26, Alpha=1, GravityOverride:Vector="0,0,0", ExternalForce="0,0,0") -> (Pose) {
    angularDampingOverride=0.995
    angularSpringConstant=2000
    bAngularSpring=true
    bOverrideAngularDamping=true
    bUseGravityOverride=true
    simulationSpace=RootRelative
    alphaBoolBlend {blendOption=Linear}
    boundBone {boneName=dyn_hair_020}
    componentAppliedLinearAccClamp=<5000 5000 5000>
    componentLinearAccScale=<0.6 0.6 0.6>
    componentLinearVelScale=<0.7 0.7 0.7>
    physicsBodyDefinitions [{
      boundBone {boneName=dyn_hair_020}
      boxExtents=<16 16 16>
      constraintSetup {angularLimitsMax=<0 0 8>, angularLimitsMin=<0 -10 -8>, angularTarget=<1 0 0>}
      localJointOffset=<0.5 0 0>
     }]
   }
   n28 = Anim Dynamics - Bone: dyn_hair_21!(ComponentPose=@n27, Alpha=1, GravityOverride:Vector="0,0,0", ExternalForce="0,0,0") -> (Pose) {
    angularDampingOverride=0.995
    angularSpringConstant=2000
    bAngularSpring=true
    bOverrideAngularDamping=true
    bUseGravityOverride=true
    simulationSpace=RootRelative
    alphaBoolBlend {blendOption=Linear}
    boundBone {boneName=dyn_hair_21}
    componentAppliedLinearAccClamp=<5000 5000 5000>
    componentLinearAccScale=<0.6 0.6 0.6>
    componentLinearVelScale=<0.7 0.7 0.7>
    physicsBodyDefinitions [{
      boundBone {boneName=dyn_hair_21}
      boxExtents=<16 16 16>
      constraintSetup {angularLimitsMax=<0 0 8>, angularLimitsMin=<0 -30 -8>, angularTarget=<1 0 0>}
      localJointOffset=<0.5 0 0>
     }]
   }
   Output Pose(Result=Component To Local(ComponentPose=@n28)) {layerGroup=None}
  ]}
 EventGraph {nodes [
   Event BlueprintUpdateAnimation()
   > n1 = Gravity Override State Function DEPRECATED(
    CharacterPartAnimInstance=Self-Reference()
    GameTimeInSeconds=GetGameTimeInSeconds()
    GravityOverrideParams=Get Gravity Override Params()
    RBAN Params=Get RBAN Params()
   )
   > Set Gravity Override(Gravity Override=@n1)
  ]}
]
variables [
 Gravity Override {type=Vector}
 Gravity Override Params {type=GravityOverrideParamsStruct}
 RBAN Params {
  type=RBANParamsStruct
  default {
   dBNO {simSpaceSettings {
     maxAngularAcceleration=1000
     maxAngularVelocity=3
     maxLinearAcceleration=300
     maxLinearVelocity=700
     velocityScaleZ=0
     externalLinearDragV=<0 0.5 0>
    }}
   dBNO_Carried {simSpaceSettings {
     maxAngularAcceleration=1000
     maxAngularVelocity=3
     maxLinearAcceleration=300
     maxLinearVelocity=700
     velocityScaleZ=0
     externalLinearDragV=<0 0.5 0>
    }}
   emoteOrMelee {simSpaceSettings {
     maxAngularAcceleration=1000
     maxAngularVelocity=3
     maxLinearAcceleration=300
     maxLinearVelocity=700
     velocityScaleZ=0
     externalLinearDragV=<0 0.5 0>
    }}
   front_End {simSpaceSettings {
     maxAngularAcceleration=1000
     maxAngularVelocity=3
     maxLinearAcceleration=300
     maxLinearVelocity=700
     velocityScaleZ=0
     externalLinearDragV=<0 0.5 0>
    }}
   onGround_Crouch_Moving {simSpaceSettings {
     maxAngularAcceleration=1000
     maxAngularVelocity=3
     maxLinearAcceleration=300
     maxLinearVelocity=700
     velocityScaleZ=0
     externalLinearDragV=<0 0.5 0>
    }}
   onGround_Falling {simSpaceSettings {
     maxAngularAcceleration=1000
     maxAngularVelocity=3
     maxLinearAcceleration=300
     maxLinearVelocity=700
     velocityScaleZ=0
     externalLinearDragV=<0 0.5 0>
    }}
   onGround_Moving {simSpaceSettings {
     maxAngularAcceleration=1000
     maxAngularVelocity=3
     maxLinearAcceleration=300
     maxLinearVelocity=700
     velocityScaleZ=0
     externalLinearDragV=<0 0.5 0>
    }}
   onGround_Standing {simSpaceSettings {
     maxAngularAcceleration=1000
     maxAngularVelocity=3
     maxLinearAcceleration=300
     maxLinearVelocity=700
     velocityScaleZ=0
     externalLinearDragV=<0 0.5 0>
    }}
   rail_Grinding {simSpaceSettings {
     maxAngularAcceleration=1000
     maxAngularVelocity=3
     maxLinearAcceleration=300
     maxLinearVelocity=700
     velocityScaleZ=0
     worldAlpha=0.1
     externalLinearDragV=<0 0.5 0>
    }}
   riding_Moving {simSpaceSettings {
     maxAngularAcceleration=1000
     maxAngularVelocity=3
     maxLinearAcceleration=300
     maxLinearVelocity=700
     velocityScaleZ=0
     externalLinearDragV=<0 0.5 0>
    }}
   riding_Moving_Targeting {simSpaceSettings {
     maxAngularAcceleration=1000
     maxAngularVelocity=3
     maxLinearAcceleration=300
     maxLinearVelocity=700
     velocityScaleZ=0
     externalLinearDragV=<0 0.5 0>
    }}
   riding_Standing {simSpaceSettings {
     maxAngularAcceleration=1000
     maxAngularVelocity=3
     maxLinearAcceleration=300
     maxLinearVelocity=700
     velocityScaleZ=0
     externalLinearDragV=<0 0.5 0>
    }}
   skydive_DivingDown {simSpaceSettings {
     maxAngularAcceleration=1000
     maxAngularVelocity=3
     maxLinearAcceleration=300
     maxLinearVelocity=700
     velocityScaleZ=0
     externalLinearDragV=<0 0.5 0>
    }}
   skydive_FloatingDown {
    componentLinearAccScale=<0 0.5 0>
    componentLinearVelScale=<0 0.5 0>
    simSpaceSettings {
     maxAngularAcceleration=1000
     maxAngularVelocity=3
     maxLinearAcceleration=300
     maxLinearVelocity=700
     velocityScaleZ=0
     externalLinearDragV=<0 0.5 0>
    }
   }
   skydive_GoingUp {simSpaceSettings {
     maxAngularAcceleration=1000
     maxAngularVelocity=3
     maxLinearAcceleration=300
     maxLinearVelocity=700
     velocityScaleZ=0
     externalLinearDragV=<0 0.5 0>
    }}
   skydive_Parachuting {
    componentLinearAccScale=<0 0.1 0.3>
    componentLinearVelScale=<0 0.1 0.3>
    simSpaceSettings {
     maxAngularAcceleration=1000
     maxAngularVelocity=3
     maxLinearAcceleration=300
     maxLinearVelocity=700
     velocityScaleZ=0
     externalLinearDragV=<0 0.5 0>
    }
   }
   skydive_Surfing {simSpaceSettings {
     maxAngularAcceleration=1000
     maxAngularVelocity=3
     maxLinearAcceleration=300
     maxLinearVelocity=700
     velocityScaleZ=0
     externalLinearDragV=<0 0.5 0>
    }}
   swimming {simSpaceSettings {
     maxAngularAcceleration=1000
     maxAngularVelocity=3
     maxLinearAcceleration=300
     maxLinearVelocity=700
     velocityScaleZ=0
     externalLinearDragV=<0 0.5 0>
    }}
   tactical_Sprint {simSpaceSettings {
     maxAngularAcceleration=1000
     maxAngularVelocity=3
     maxLinearAcceleration=300
     maxLinearVelocity=700
     velocityScaleZ=0
     externalLinearDragV=<0 0.5 0>
    }}
   ziplining {simSpaceSettings {
     maxAngularAcceleration=1000
     maxAngularVelocity=3
     maxLinearAcceleration=300
     maxLinearVelocity=700
     velocityScaleZ=0
     externalLinearDragV=<0 0.5 0>
    }}
  }
 }
]
